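# Winterfest Season 33 quest conversation in which an NPC briefs the player on tampered basketball equipment and sends them to investigate Hopeful Heights.
# A single entry tag starts a linear chain of three speech lines, ending in a Continue node with the acceptance reply.
# Each speech requires the NPC's GumshoeDark conversation tag and the active P2_Q01 quest.
# The third line's choice carries the Quest.01 tag, while the final choice carries Quest.02.
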ConversationDatabase /BattlePassS33_Quests/Conversations/Winterfest/Case_2/Quest_S33_Winterfest_P2_Q01
entry_points [{tag=Conversation.Entry.Season33.WF.Quest.02.part.01, destinations [#1]}]
nodes [
 ConversationEntryPointNode_1 {
  id=#1
  class=ConversationEntryPointNode
  outputs [#2]
  properties {entryTag {tagName=Conversation.Entry.Season33.WF.Quest.02.part.01}}
 }
 Speech_C_1 {
  id=#2
  class=Speech_C
  outputs [#3]
  properties {softTaskIcon=&T_UI_NPC_Intel2, generalConfig {defaultMessage="We have a new Naughty List candidate, rookie."}}
  sub_nodes [
   ConversationChoiceNode_6 {class=ConversationChoiceNode, properties {defaultChoiceDisplayText=Wintervestigation}}
   FortConversationRequirement_ControllerRequirement_10 {
    class=FortConversationRequirement_ControllerRequirement
    properties {
     nodeName=Require: NPC Has Conversation Tag
     participantId {tagName=Conversation.Participant.NPC}
     requirement {class=FortControllerRequirement_NPCHasConversationEntryTag, properties {queryTag {tagName=Conversation.Entry.GumshoeDark}}}
    }
   }
   FortConversationRequirement_ControllerRequirement_11 {
    class=FortConversationRequirement_ControllerRequirement
    properties {
     nodeName=Require: Is Quest Active
     requirement {
      class=FortControllerRequirement_IsQuestActive
      properties {questDefinition=/BattlePassS33_Quests/Items/QuestItems/Winterfest/Case_2/Quest_S33_Winterfest_P2_Q01}
     }
    }
   }
  ]
 }
 Speech_C_2 {
  id=#3
  class=Speech_C
  outputs [#4]
  properties {
   softTaskIcon=&T_UI_NPC_Intel2
   generalConfig {defaultMessage=We're getting reports someone has been tampering with the local basketball equipment}
  }
  sub_nodes [
   ConversationChoiceNode_7 {class=ConversationChoiceNode, properties {defaultChoiceDisplayText=...}}
   FortConversationRequirement_ControllerRequirement_12 {
    class=FortConversationRequirement_ControllerRequirement
    properties {
     nodeName=Require: NPC Has Conversation Tag
     participantId {tagName=Conversation.Participant.NPC}
     requirement {class=FortControllerRequirement_NPCHasConversationEntryTag, properties {queryTag {tagName=Conversation.Entry.GumshoeDark}}}
    }
   }
   FortConversationRequirement_ControllerRequirement_13 {
    class=FortConversationRequirement_ControllerRequirement
    properties {
     nodeName=Require: Is Quest Active
     requirement {
      class=FortControllerRequirement_IsQuestActive
      properties {questDefinition=&Quest_S33_Winterfest_P2_Q01}
     }
    }
   }
  ]
 }
 Speech_C_3 {
  id=#4
  class=Speech_C
  outputs [#5]
  properties {softTaskIcon=&T_UI_NPC_Intel2, generalConfig {defaultMessage=Go out to Hopeful Heights and investigate.}}
  sub_nodes [
   ConversationChoiceNode_8 {
    class=ConversationChoiceNode
    properties {defaultChoiceDisplayText=..., choiceTags {gameplayTags [{tagName=Conversation.Entry.Season33.WF.Quest.01.part.01}]}}
   }
   FortConversationRequirement_ControllerRequirement_14 {
    class=FortConversationRequirement_ControllerRequirement
    properties {
     nodeName=Require: NPC Has Conversation Tag
     participantId {tagName=Conversation.Participant.NPC}
     requirement {class=FortControllerRequirement_NPCHasConversationEntryTag, properties {queryTag {tagName=Conversation.Entry.GumshoeDark}}}
    }
   }
   FortConversationRequirement_ControllerRequirement_15 {
    class=FortConversationRequirement_ControllerRequirement
    properties {
     nodeName=Require: Is Quest Active
     requirement {
      class=FortControllerRequirement_IsQuestActive
      properties {questDefinition=&Quest_S33_Winterfest_P2_Q01}
     }
    }
   }
  ]
 }
 Continue_C_1 {
  id=#5
  class=Continue_C
  properties {nodeName="", softTaskIcon=&T_UI_NPC_Intel2}
  sub_nodes [ConversationChoiceNode_9 {
    class=ConversationChoiceNode
    properties {
     defaultChoiceDisplayText="Now, THIS sounds actually naughty. On it!"
     choiceTags {gameplayTags [{tagName=Conversation.Entry.Season33.WF.Quest.02.part.01}]}
    }
   }]
 }
]
