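# NPC conversation database for part 5 of the Season 33 Winterfest Case 2 quest, where the player reports to an NPC tagged GumshoeDark.
# The "Wintervestigation" choice requires that NPC tag and the quest to be active.
# A linear chain runs from the entry point through two speech exchanges to a closing continue node, whose choice carries the conversation's entry tag as a choice tag.
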ConversationDatabase /BattlePassS33_Quests/Conversations/Winterfest/Case_2/Quest_S33_Winterfest_P2_Q05
entry_points [{tag=Conversation.Entry.Season33.WF.Quest.02.part.05, destinations [#1]}]
nodes [
 ConversationEntryPointNode_1 {
  id=#1
  class=ConversationEntryPointNode
  outputs [#2]
  properties {entryTag {tagName=Conversation.Entry.Season33.WF.Quest.02.part.05}}
 }
 Speech_C_1 {
  id=#2
  class=Speech_C
  outputs [#3]
  properties {softTaskIcon=&T_UI_NPC_Intel2, generalConfig {defaultMessage="Rookie. Report!"}}
  sub_nodes [
   ConversationChoiceNode_10 {class=ConversationChoiceNode, properties {defaultChoiceDisplayText=Wintervestigation}}
   FortConversationRequirement_ControllerRequirement_16 {
    class=FortConversationRequirement_ControllerRequirement
    properties {
     nodeName=Require: NPC Has Conversation Tag
     participantId {tagName=Conversation.Participant.NPC}
     requirement {class=FortControllerRequirement_NPCHasConversationEntryTag, properties {queryTag {tagName=Conversation.Entry.GumshoeDark}}}
    }
   }
   FortConversationRequirement_ControllerRequirement_17 {
    class=FortConversationRequirement_ControllerRequirement
    properties {
     nodeName=Require: Is Quest Active
     requirement {
      class=FortControllerRequirement_IsQuestActive
      properties {questDefinition=/BattlePassS33_Quests/Items/QuestItems/Winterfest/Case_2/Quest_S33_Winterfest_P2_Q05}
     }
    }
   }
  ]
 }
 Speech_C_2 {
  id=#3
  class=Speech_C
  outputs [#4]
  properties {softTaskIcon=&T_UI_NPC_Intel2, generalConfig {defaultMessage=Hm. That's concerning. I'll look into it.}}
  sub_nodes [ConversationChoiceNode_11 {class=ConversationChoiceNode, properties {defaultChoiceDisplayText="That's the second nice person you've sent me after!"}}]
 }
 Continue_C_1 {
  id=#4
  class=Continue_C
  properties {nodeName="", softTaskIcon=&T_UI_NPC_Intel2}
  sub_nodes [ConversationChoiceNode_13 {
    class=ConversationChoiceNode
    properties {defaultChoiceDisplayText=You do that., choiceTags {gameplayTags [{tagName=Conversation.Entry.Season33.WF.Quest.02.part.05}]}}
   }]
 }
]
